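# Physics asset for the M_MED_Informer backpack cosmetic, defining bodies and joints for its dangling accessory bones.
# A compound pack body of two boxes and a capsule anchors the sword, flap, zipper and four droid bodies, each constrained directly to it.
# The flap hinges on one swing axis with the others locked, the sword has tight angular limits, and the zipper and droids swing widely.
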PhysicsAsset /BRCosmetics/Accessories/FORT_Backpacks/Backpack_M_MED_Informer/Meshes/M_MED_Informer_Pack_dyn
bodies {
 dyn_droids_01 [{shapeType=Capsule, length=2, radius=3.3, center=<1 0 0>, rotation=<rot 90 0 0>}]
 dyn_droids_02 [{shapeType=Capsule, length=2, radius=3.3, center=<1 0.5 0>, rotation=<rot 90 0 0>}]
 dyn_droids_03 [{shapeType=Capsule, length=2, radius=3.3, center=<0.8 -0.5 0>, rotation=<rot 90 0 0>}]
 dyn_droids_04 [{shapeType=Capsule, length=2, radius=3.3, center=<0.5 0 0>, rotation=<rot 90 0 0>}]
 dyn_flap [{shapeType=Capsule, length=20, radius=1.3, center=<10 0 0>, rotation=<rot 0 0 90>}]
 dyn_sword [{shapeType=Box, extentX=30, extentY=7, extentZ=4, center=<5 0 1.8>}]
 dyn_zipper [{shapeType=Capsule, length=1.5, radius=0.8, center=<1.2 0 0>, rotation=<rot 90 0 0>}]
 pack [
  {shapeType=Capsule, length=11.3229, radius=5.6906, center=<-4.6625 -13.5782 -15.0503>}
  {shapeType=Box, extentX=21.8875, extentY=12.5065, extentZ=24.8818, center=<-0.2042 -6.2547 -13.2097>}
  {shapeType=Box, extentX=21.8875, extentY=12.0464, extentZ=18.0259, center=<-0.2042 -3.2413 0.3712>, rotation=<rot 0 0 29.9999>}
 ]
}
constraints [
 {bone1Name=dyn_sword, bone2Name=pack, swing1LimitDegrees=15, swing2LimitDegrees=5, twistLimitDegrees=10}
 {
  bone1Name=dyn_flap
  bone2Name=pack
  swing1LimitDegrees=2
  swing1Motion=Locked
  swing2LimitDegrees=45
  twistLimitDegrees=5
  twistMotion=Locked
 }
 {bone1Name=dyn_zipper, bone2Name=pack, swing1LimitDegrees=75, swing2LimitDegrees=75, twistLimitDegrees=5}
 {bone1Name=dyn_droids_01, bone2Name=pack, swing1LimitDegrees=60, swing2LimitDegrees=60, twistLimitDegrees=30}
 {bone1Name=dyn_droids_02, bone2Name=pack, swing1LimitDegrees=60, swing2LimitDegrees=60, twistLimitDegrees=30}
 {bone1Name=dyn_droids_03, bone2Name=pack, swing1LimitDegrees=60, swing2LimitDegrees=60, twistLimitDegrees=30}
 {bone1Name=dyn_droids_04, bone2Name=pack, swing1LimitDegrees=60, swing2LimitDegrees=60, twistLimitDegrees=30}
]
